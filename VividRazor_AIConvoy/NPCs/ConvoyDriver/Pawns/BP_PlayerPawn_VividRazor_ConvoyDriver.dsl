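; AI convoy leader pawn that spawns follower point providers behind itself and manages the convoy's bike usage.
; On the server, a timer polls for a patrol path, logging and giving up after five seconds.
; Once a path exists, it places providers at alternating side offsets, projected onto the navmesh, then requests Living World event generation.
; As a workaround, a reinforcements gameplay event tags tracked follower vehicles as AI-disabled, preventing remounting.
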(event EventBeginPlay
  (|CallParentFunction)
  (Networking|SwitchHasAuthority
    (:Authority
      (Variables|InternalData|SetActorSpawnTime (Utilities|Time|GetGameTimeinSeconds))
      (CallFunction|WaitforPatrolPathBeingSet)
      (Ability|Async|WaitGameplayEventToActor 0 (Variables|Default|GetEventTag_ReinforcementsStarted) true
        (:EventReceived
          (CallFunction|DisableBikeUsage))))))

(event Custom|WaitForPatrolPathBeingSet
  ; Before we can spawn Convoy Followers, we need to wait for the AI to have a Path set on them.
  (Utilities|IsValid (CallFunction|GetPatrolPath)
    (:Is%20Valid
      (Utilities|Time|ClearandInvalidateTimerbyHandle (Variables|InternalData|GetHDNL_ValidatePatrolPathTimer))
      (CallFunction|FindLocationstoSpawnFollowers))
    (:Is%20Not%20Valid
      (Utilities|FlowControl|Sequence
        (:then_0
          (if (>= (- (Utilities|Time|GetGameTimeinSeconds) (Variables|InternalData|GetActorSpawnTime)) 5.0)
            (Fortnite|FortShippingLog self "Convoy Leader was not assigned a Patrol Path after 5 seconds. Trying to stop. This Leader will not have Followers." true)
            (Utilities|Time|ClearandInvalidateTimerbyHandle (Variables|InternalData|GetHDNL_ValidatePatrolPathTimer))))
        (:then_1
          (Utilities|FlowControl|DoOnce false
            (:Completed
              (bind _returnvalue (Utilities|Time|SetTimerbyEvent (EventDispatchers|CreateEvent :!event Custom|WaitForPatrolPathBeingSet) 0.1 true))
              (Variables|InternalData|SetHDNL_ValidatePatrolPathTimer _returnvalue))))))))

(event Custom|FindLocationsToSpawnFollowers
  (bind _returnvalue (Class|FortAthenaLivingWorldManager|GetLivingWorldManager))
  ; Get some Locations to spawn the Convoy Followers
  (for _index (range 1 (+ (Math|Float|Truncate (ScalableFloat|GetValueatLevel (Variables|Default|GetNumberOfFollowers))) 1))
    (bind _location (CallFunction|CalculateSpawnLocation _index))
    (Game|SpawnActorfromClass "/VividRazor_AIConvoy/Blueprints/BP_PointProviderStatic_ConvoyFollower.BP_PointProviderStatic_ConvoyFollower_C" _location "AlwaysSpawn" "MultiplyWithRoot" 0 self))
  (Utilities|IsValid _returnvalue
    (:Is%20Valid
      (Class|FortAthenaLivingWorldManager|RequestEventGeneration _returnvalue))))

(event Custom|TrackFollowerVehicles
  (params (Actor :type Actor%20Object%20Reference))
  ; Work Around to tell AI Bikes NOT to go back on their Bikes. Need a better Evaluator with more flexible options in the future! :)
  (Utilities|Array|AddUnique (Variables|Default|GetVehicleActors) Actor))

(event Custom|DisableBikeUsage
  ; Work Around to tell AI Bikes NOT to go back on their Bikes. Need a better Evaluator with more flexible options in the future! :)
  (for _array_element (Variables|Default|GetVehicleActors)
    (Ability|GameplayEffect|AddLooseGameplayTags _array_element (Variables|Default|GetGamepalyTags_AIDisabledVehicle) true)))

(fn CalculateSpawnLocation
  (params (Index :type Integer))
  (bind (_projectedlocation _returnvalue) (AI|Navigation|ProjectPointtoNavigation (+ (+ (Transformation|GetActorLocation self) (* (Transformation|GetActorForwardVector self) (* (ScalableFloat|GetValueatLevel (Variables|Default|GetFollowerSpawnDistance)) -1.0))) (* (Transformation|GetActorRightVector self) (Utilities|Select :Index (Math|Integer|%25%28Integer%29 (Variables|GetIndex) 2) :Option%200 (ScalableFloat|GetValueatLevel (Variables|Default|GetFollowerOffsetSpawnDistance)) :Option%201 (* (ScalableFloat|GetValueatLevel (Variables|Default|GetFollowerOffsetSpawnDistance)) -1.0)))) (Fortnite|AI|GetNavigationDataforActor self) "/Game/Athena/AI/Phoebe/NavFilter/NavFilter_Phoebe_NoSmashableNoWater.NavFilter_Phoebe_NoSmashableNoWater_C" "1200.000000,1200.000000,1200.000000"))
  (return (Math|Transform|MakeTransform (select _returnvalue _projectedlocation (+ (+ (Transformation|GetActorLocation self) (* (Transformation|GetActorForwardVector self) (* (ScalableFloat|GetValueatLevel (Variables|Default|GetFollowerSpawnDistance)) -1.0))) (* (Transformation|GetActorRightVector self) (Utilities|Select :Index (Math|Integer|%25%28Integer%29 (Variables|GetIndex) 2) :Option%200 (ScalableFloat|GetValueatLevel (Variables|Default|GetFollowerOffsetSpawnDistance)) :Option%201 (* (ScalableFloat|GetValueatLevel (Variables|Default|GetFollowerOffsetSpawnDistance)) -1.0))))) (Transformation|GetActorRotation self))))

(fn GetPatrolPath
  (bind _asbp_phoebe_controller_non_participant (Utilities|Casting|CastToBP_PhoebeController_NonParticipant (Pawn|GetController self))
    (:then
      (Utilities|IsValid (Actor|GetComponentbyClass _asbp_phoebe_controller_non_participant "/Script/FortniteAI.FortAthenaNpcPatrollingComponent")
        (:Is%20Valid
          (return (Class|FortAthenaNpcPatrollingComponent|GetPatrolPath (Actor|GetComponentbyClass _asbp_phoebe_controller_non_participant "/Script/FortniteAI.FortAthenaNpcPatrollingComponent") :!class /Script/FortniteAI.FortAthenaNpcPatrollingComponent)))
        (:Is%20Not%20Valid
          _)))
    (:CastFailed
      _))
  (return 0))

(fn ConstructionScript
  (|CallParentFunction))
